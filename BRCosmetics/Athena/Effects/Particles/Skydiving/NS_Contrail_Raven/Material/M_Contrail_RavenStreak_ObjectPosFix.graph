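# Unlit translucent Niagara ribbon material for the Raven skydiving contrail streak.
# Layers panning soft-smoke and inky-trail textures, with dynamic parameters shifting UVs and scaling pan speed.
# Color blends two dark tints by smoke, darkens along U, then scales by time of day and particle color.
# Opacity combines lighten-blended smoke, edge sphere masks, a clamped sine pulse, camera-proximity fade and particle alpha, with depth fade skipped on mobile.
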Material /BRCosmetics/Athena/Effects/Particles/Skydiving/NS_Contrail_Raven/Material/M_Contrail_RavenStreak_ObjectPosFix
graph {
 properties {
  bAutomaticallySetUsageInEditor=false
  bUsedWithNiagaraRibbons=true
  bUsedWithStaticMesh=false
  blendMode=BLEND_Translucent
  pixelDepthOffsetMode=PDOM_Legacy
  shadingModel=MSM_Unlit
  usedShadingModels=Unlit
 }
 nodes [
  n1 = VectorParameter() {parameterName=BaseColor_A, defaultValue=<rgba 0.0083 0 0.0156 1>}
  n2 = VectorParameter() {parameterName=BaseColor_B, defaultValue=<rgba 0 0.0009 0.0208 1>}
  n3 = DynamicParameter() {paramNames [AlphaMask_U, RandomOffset, xPanSpeedMult, 0]}
  n4 = MaterialFunctionCall(Float2=TextureCoordinate() {coordinateIndex=1}) {materialFunction=&BreakOutFloat2Components}
  n5 = Add(A=Multiply(A=Add(A=@n4.R, B=@n3.AlphaMask_U)) {constB=7}, B=@n3.RandomOffset)
  n6 = Multiply(A=ComponentMask(None=TextureCoordinate()) {g=true})
  n7 = ScalarParameter() {defaultValue=1, parameterName=TimeMutiply}
  n8 = Multiply(A=Time(), B=@n7)
  n9 = AppendVector(A=@n3.xPanSpeedMult, B=Constant())
  n10 = TextureSample(UVs=Panner(Coordinate=AppendVector(A=@n5, B=@n6), Time=@n8, Speed=@n9) {speedX=-2}) {samplerType=SAMPLERTYPE_Grayscale, texture=&T_Smoke_Trail_Soft}
  n11 = ComponentMask(None=TextureCoordinate()) {r=true}
  n12 = TextureCoordinate()
  n13 = AppendVector(A=Multiply(A=@n4.R) {constB=22}, B=Multiply(A=ComponentMask(None=@n12) {g=true}))
  n14 = Time()
  n15 = TextureSample(UVs=Panner(Coordinate=@n13, Time=Multiply(A=@n14, B=@n7), Speed=@n9)) {texture=/BRCosmetics/Athena/Effects/Textures/T_InkySmokeTrail}
  n16 = Multiply(A=Add(A=@n15.R, B=@n15.R)) {constB=3}
  n17 = Multiply(A=@n10.R, B=@n10.R)
  n18 = Multiply(
   A=LinearInterpolate(A=LinearInterpolate(A=@n1.RGB, B=@n2.RGB, Alpha=@n10.R), B=Constant(), Alpha=Multiply(A=@n11, B=@n11))
   B=Multiply(A=Add(A=@n16, B=@n17)) {constB=4}
  ) {constA=1}
  n19 = Constant() {r=1.2}
  n20 = Constant() {r=0.5}
  n21 = MaterialFunctionCall(Color_Morning=@n19, Color_Day=@n19, Color_Evening=@n20, Color_Night=@n20) {materialFunction=&Color_Time_of_Day}
  n22 = ParticleColor()
  MaterialFunctionCall(Emissive=Multiply(A=Multiply(A=@n18, B=@n21.Result), B=@n22.RGB)) {materialFunction=&MF_FX_EmissiveAdjustments}
  n23 = SphereMask(
   A=MaterialFunctionCall(Float2=TextureCoordinate()) {materialFunction=&BreakOutFloat2Components}.G
   B=Constant() {r=0.5}
  ) {attenuationRadius=0.5, hardnessPercent=0}
  n24 = TextureCoordinate()
  n25 = SphereMask(
   A=Power(Base=MaterialFunctionCall(Float2=@n24) {materialFunction=&BreakOutFloat2Components}.R) {constExponent=0.15}
   B=Constant() {r=0.5}
  ) {attenuationRadius=0.5, hardnessPercent=0}
  Multiply(A=@n23, B=Multiply(A=@n25, B=@n25))
  n26 = Sine(None=Multiply(A=Constant() {r=0.22}, B=Time()))
  n27 = Time()
  n28 = TextureSample(UVs=Panner(Coordinate=@n13, Time=Multiply(A=@n27, B=@n7), Speed=@n9) {speedX=-5}) {texture=&T_InkySmokeTrail}
  n29 = Add(A=@n3.AlphaMask_U, B=@n4.R)
  n30 = Add(A=@n3.RandomOffset, B=Multiply(A=Add(A=@n29) {constB=0.6}) {constB=9})
  n31 = TextureSample(UVs=Panner(Coordinate=AppendVector(A=@n30, B=@n6), Time=@n8, Speed=@n9) {speedX=-2}) {samplerType=SAMPLERTYPE_Grayscale, texture=&T_Smoke_Trail_Soft}
  n32 = MaterialFunctionCall!(
   Blend=Multiply(A=Add(A=@n28.G, B=@n28.G)) {constB=1.1}
   Base=Multiply(A=MaterialFunctionCall(Blend=@n10.G, Base=@n31.G) {materialFunction=/Engine/Functions/Engine_MaterialFunctions03/Blends/Blend_Lighten}.Result) {constB=1.25}
  ) {materialFunction=&Blend_Lighten}
  n33 = Multiply(A=@n23, B=SphereMask(A=TextureCoordinate(), B=Constant2Vector() {g=0.5, r=1.025}) {attenuationRadius=1.05, hardnessPercent=90})
  n34 = SphereMask(A=WorldPosition().XYZ, B=CameraPositionWS(), Radius=Constant() {r=100}, Hardness=Constant() {r=0.2})
  n35 = Multiply(A=Clamp(None=Abs(None=@n26)) {minDefault=0.6}, B=Multiply(A=Multiply(A=@n32.Result, B=@n33), B=OneMinus(None=@n34)))
  Multiply(A=@n22.A, B=ShadingPathSwitch(Default=DepthFade(Opacity=@n35), Mobile=@n35) {desc=Migrated from FeatureLevelSwitch})
 ]
}
