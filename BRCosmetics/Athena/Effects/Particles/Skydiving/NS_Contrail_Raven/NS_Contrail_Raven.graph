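# Niagara skydiving contrail for the Raven cosmetic, built from ribbon trails, feather meshes and glows, all attached to skeletal mesh positions.
# Four short-lived ribbons flash bright then fade, with sampled positions fed to material parameters.
# Feather meshes spawn at sampled bones, tumbling and orbiting outward as they age.
# Four persistent purple glow sprites follow the bones with per-particle curve-driven flicker, while front-end modifiers swap motion vectors using User.FEVec.
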NiagaraSystem /BRCosmetics/Athena/Effects/Particles/Skydiving/NS_Contrail_Raven/NS_Contrail_Raven
dependencies {usedDataInterfaces [&NiagaraDataInterfaceCurve], usedDynamicInputs [], usedModules [&SystemState]}
emitters {
 FN_Contrails_Ribbon_v4 {
  emitterUpdateScript [
   &EmitterScalability(
    Module.bApplySpawnCountScale=false
    Module.bApplySpawnCountScaleByDistanceFraction=false
    Module.bUseMaxDistance=false
    Module.bUseMinDistance=false
   ) {enabled=false}
   &EmitterLifeCycle() {enabled=false}
   &EmitterState()
   /Game/Effects/Niagara/Modules/Spawn/FN_SpawnContrails(Global Spawnrate=30) {enabled=false}
   &FN_SpawnRibbon_v2(Number of Ribbons=4, Spawn Mode=::NewEnumerator0, SpawnRate=30)
  ]
  particleSpawnScript [
   &FN_InitializeRibbon_v3(Enable Skeletal Position=true, Lifetime=1, Ribbon Width=20)
   &AddVelocity(Velocity=&FN_FrontEndVector3Modifier(Gameplay=<0 0 0>, Front End=&Multiply_VectorByFloat(Vector=$User.FEVec, Float=5)))
   /SetVariables(Particles.RandomFloat=&UniformRangedFloat(Minimum=0, Maximum=1, Randomness Mode=::NewEnumerator1, Random Seed=$Particles.SpawnGroup, Fixed Random Seed=true)) {isSetParameters=true}
  ]
  particleUpdateScript [
   /Game/Effects/Niagara/Modules/Skydiving/FN_DivingAttributes/FN_DivingAttributes()
   &SolveForcesAndVelocity(Clamp Velocity=false, Limit Acceleration=false)
   &FN_UpdateRibbon(
    Enable Skeletal Position=true
    Initial Color:LinearColor=$Particles.Initial.Color
    Initial RibbonWidth:NiagaraFloat=$Particles.Initial.RibbonWidth
   )
   /Game/Effects/Niagara/Modules/Materials/MaterialPanner(Jitter Multiplier=3.9, Speed=1.5)
   &Material(Base Texture Time:NiagaraFloat=&Multiply_Float(A=$Emitter.Age, B=-1))
   /SetVariables(Particles.PositionShift:Vector3f=&Add_Vector(
     A=Linked() {linkedVariable {name=Particles.Position, type=NiagaraPosition}}
     B=&FN_FrontEndVector3Modifier(
      Gameplay=<0 0 -5>
      Front End=&Add_Vector(
       A=<0 0 -6>
       B=&Multiply_Vector(A=&TransformVector(Vector=$User.FEVec, Source Space=::World, Destination Space=::Local), B=<0 -0.05 0>)
      )
     )
    )) {isSetParameters=true}
   /Niagara/Modules/Update/Material/DynamicMatParams1(
    Module.Param0WriteEnabled=true
    Module.Param1WriteEnabled=true
    Module.Param2WriteEnabled=true
    Module.Param3WriteEnabled=true
    Param0Value:NiagaraFloat=&MakeFloatFromVector(Vector=$Particles.FN_UpdateRibbon.SampledPosition, Channel=::NewEnumerator0)
    Param1Value:NiagaraFloat=&MakeFloatFromVector(Vector=$Particles.FN_UpdateRibbon.SampledPosition, Channel=::NewEnumerator1)
    Param2Value:NiagaraFloat=&MakeFloatFromVector(Vector=$Particles.FN_UpdateRibbon.SampledPosition, Channel=::NewEnumerator2)
   ) {enabled=false}
   /Niagara/Modules/Update/Material/DynamicMatParams(
    Module.Param0WriteEnabled=true
    Module.Param1WriteEnabled=true
    Module.Param2WriteEnabled=true
    Module.Param3WriteEnabled=true
    Param0Value=0.5
    Param1Value:NiagaraFloat=$Particles.RandomFloat
    Param2Value:NiagaraFloat=&FloatFromCurve(FloatCurve=DataInterface() {propertyValues {Curve {keys [
         <at -0.0046 1.37>
         {arriveTangent=-3.8373, interpMode=RCIM_Cubic, leaveTangent=-3.8373, time=0.22, value=0.77}
         {arriveTangent=0.0538, interpMode=RCIM_Cubic, leaveTangent=0.0538, tangentMode=RCTM_User, time=0.6, value=-0.95}
         <at 1.0011 0.002>
        ]}}})
    Param3Value=1
   ) {enabled=false}
   &DynamicMaterialParameters(
    Index 0 Param 1=0.5
    Index 0 Param 2=$Particles.RandomFloat
    Index 0 Param 3=-1.7
    Index 0 Param 4=1
    Index 1 Param 1=&MakeFloatFromVector(Vector=$Particles.FN_UpdateRibbon.SampledPosition, Channel=::NewEnumerator0)
    Index 1 Param 2=&MakeFloatFromVector(Vector=$Particles.FN_UpdateRibbon.SampledPosition, Channel=::NewEnumerator1)
    Index 1 Param 3=&MakeFloatFromVector(Vector=$Particles.FN_UpdateRibbon.SampledPosition, Channel=::NewEnumerator2)
    Module.Param0WriteEnabled=true
    Module.Param0WriteEnabled1=true
    Module.Param1WriteEnabled=true
    Module.Param1WriteEnabled1=true
    Module.Param2WriteEnabled=true
    Module.Param2WriteEnabled1=true
    Module.Param3WriteEnabled=true
    Module.Param3WriteEnabled1=true
    Write Parameter Index 1=true
   )
   &Color(
    Scale Alpha=&FloatFromCurve(
     FloatCurve=DataInterface() {propertyValues {Curve {keys [<at 0 0>,<at 0.05 1>,<at 0.15 0.3696>,<at 0.3 0>]}}}
     Scale Curve=&FN_FrontEndFloatModifier(Gameplay=1, Front End=1)
    )
    Scale Color=&VectorFromFloat(Value=&FloatFromCurve(FloatCurve=DataInterface() {propertyValues {Curve {keys [<at 0 8>,<at 0.05 3>,<at 0.095 0>]}}}))
   )
  ]
  rendererClasses [&NiagaraRibbonRendererProperties]
  renderers [{rendererClass=&NiagaraRibbonRendererProperties}]
 }
 Feathers_ChainedInheritance {
  emitterSpawnScript [&FN_PreserveSkydiveUserVariables() {module=PreserveSkydiveUserVariables}]
  emitterUpdateScript [
   &EmitterScalability(
    MaxDistance=1000
    Module.bApplySpawnCountScale=true
    Module.bApplySpawnCountScaleByDistanceFraction=true
    Module.bUseMaxDistance=true
    Module.bUseMinDistance=false
    SpawnCountScaleByDistanceFraction=DataInterface() {propertyValues {Curve {keys [<at -0.0026 0.9982>,<at 1.1235 -0.0145>]}}}
   ) {enabled=false}
   &EmitterLifeCycle(NextLoopDuration=1) {enabled=false}
   &EmitterState()
   /Game/Effects/Niagara/Modules/Spawn/FN_SkydiveSpawnRate(Global Spawnrate=3) {module=SkydiveSpawnRate}
  ]
  particleSpawnScript [
   /Game/Effects/Niagara/Modules/Orbit/FN_Orbit_Init_Variables_v2()
   /Game/Effects/Niagara/Modules/Groups/FN_SetSpawnGroup_v1/FN_SetSpawnGroup(Override SpawnGroup=true, SpawnGroup:NiagaraInt32=&SpawnGroup())
   /Niagara/Modules/Spawn/Initialization/InitializeParticle(
    Lifetime=&UniformRangedFloat(Minimum=0.8, Maximum=1.4)
    Mesh Scale=&UniformRangedVector(Minimum=<0.075 0.15 0.15>, Maximum=<0.16 0.35 0.32>)
    Module.Write Color=true
    Module.Write Lifetime=true
    Module.Write Mass=true
    Module.Write Material Random=true
    Module.Write Position=false
    Module.Write Scale=true
    Module.Write SpriteRotation=false
    Module.Write SpriteSize=false
    Position=&SimulationPosition()
   )
   &SampleSkeletalMeshSkeleton(
    Mesh:NiagaraDataInterfaceSkeletalMesh=$User.SkeletalMesh
    Specific Bone Index:NiagaraInt32=$Particles.SpawnGroup
    Specific Socket Index:NiagaraInt32=$Particles.SpawnGroup
   )
   /Niagara/Modules/Spawn/Location/SkeletalMeshSkeletonLocation(Module.Inherit Velocity=false, Module.Use Sampled Bone as Ribbon ID=true)
   &AddVelocityInCone(
    Cone Angle=10
    Cone Axis=&FN_FrontEndVector3Modifier(Gameplay=<0 0 100>, Front End=$User.FEVec)
    Module.Use Velocity Falloff On Cone Axis=false
    Velocity Strength=100
   )
   /SetVariables(Particles.MeshOrientation=&MakeQuatFromAxisAngle(
     Axis=/Niagara/DynamicInputs/Random/RandomVector(
      Vector Scale:NiagaraFloat=1
      Constraint Vector:Vector3f=<0 0 1>
      Randomness Mode:ENiagaraRandomnessMode=::NewEnumerator0
      Random Seed:NiagaraInt32=0
      Fixed Random Seed:NiagaraBool=false
     )
     AngleInDegrees=&UniformRangedFloat(Minimum=0, Maximum=360)
    )) {isSetParameters=true}
   &MeshRotationForce(Rotation=&UniformRangedVector(Minimum=<0 0 0>, Maximum=<1 1 1>))
   &AddRotationalVelocity(Rotation Rate=&RandomVector(
     Vector Scale:NiagaraFloat=&UniformRangedFloat(Minimum=1, Maximum=2.5)
     Constraint Vector:Vector3f=<0 0 1>
     Randomness Mode:ENiagaraRandomnessMode=::NewEnumerator0
     Random Seed:NiagaraInt32=0
     Fixed Random Seed:NiagaraBool=false
    ))
   &ApplyInitialForces()
  ]
  particleUpdateScript [
   &ParticleState()
   &Color(Color=&MakeLinearColorFromVectorAndFloat(
     "Vector (RGB)"=&VectorFromFloat(Value=&FloatFromCurve(FloatCurve=DataInterface() {propertyValues {Curve {keys [<at 0 0.5>,<at 1 0.1>]}}}))
     "Float (Alpha)"=&FloatFromCurve(FloatCurve=DataInterface() {propertyValues {Curve {keys [<at -0.0046 0.5>,<at 0.4 1>,<at 1.0011 0.002>]}}})
    ))
   &DynamicMatParams(
    Module.Param0WriteEnabled=true
    Module.Param1WriteEnabled=true
    Module.Param2WriteEnabled=true
    Module.Param3WriteEnabled=true
    Param0Value:NiagaraFloat=$Particles.FeaterOffset
    Param1Value:NiagaraFloat=&FloatFromCurve(FloatCurve=DataInterface() {propertyValues {Curve {keys [<at 0.9 1.0113>,<at 1.0011 0.002>]}}})
    Param2Value:NiagaraFloat=&FloatFromCurve(FloatCurve=DataInterface() {propertyValues {Curve {keys [<at 0 10>,<at 1.0011 0.002>]}}})
   ) {enabled=false}
   &DynamicMaterialParameters(
    Index 0 Param 1=$Particles.FeaterOffset
    Index 0 Param 2=&FloatFromCurve(FloatCurve=DataInterface() {propertyValues {Curve {keys [<at 0.9 1.0113>,<at 1.0011 0.002>]}}})
    Index 0 Param 3=&FloatFromCurve(FloatCurve=DataInterface() {propertyValues {Curve {keys [<at 0 10>,<at 1.0011 0.002>]}}})
    Module.Param0WriteEnabled=true
    Module.Param1WriteEnabled=true
    Module.Param2WriteEnabled=true
    Module.Param3WriteEnabled=true
   )
   &Drag(Drag=0.1, Use Linear Drag=true, Use Rotational Drag=false)
   &SolveForcesAndVelocity(Clamp Velocity=false, Limit Acceleration=false)
   &SolveRotationalForcesAndVelocity()
   /BRCosmetics/Effects/Niagara/Modules/Orbit/FN_Orbit_v2(
    Radius min/max:Vector2f=&Multiply_Vector2(A=<10 25>, B=&Vector2DFromFloat(Value=$Particles.NormalizedAge))
    Rotation Offset min/max:Vector2f=<0 1>
    Rotation Rate min/max:Vector2f=<-1 1>
    Set Mesh Orientation=false
   ) {enabled=false}
   /Game/Effects/Niagara/Modules/Orbit/FN_Orbit_v3(
    Radius min/max:Vector2f=&Multiply_Vector2(A=<10 25>, B=&Vector2DFromFloat(Value=$Particles.NormalizedAge))
    Rotation Offset min/max:Vector2f=<0 1>
    Rotation Rate min/max:Vector2f=<-1 1>
   )
  ]
  rendererClasses [&NiagaraMeshRendererProperties]
  renderers [{rendererClass=&NiagaraMeshRendererProperties, rendererIndex=0}]
 }
 Glow_ChainedInheritance {
  emitterSpawnScript [&FN_PreserveSkydiveUserVariables() {module=PreserveSkydiveUserVariables}]
  emitterUpdateScript [
   &EmitterLifeCycle(CompleteOnInactive=true, MaxLoopCount=1, NextLoopDuration=1000) {enabled=false}
   &EmitterState(Life Cycle Mode=::NewEnumerator1, Loop Behavior=::NewEnumerator1, Loop Duration=1, UseLoopDelay=false)
   &SpawnBurst_Instantaneous(Spawn Count=4, Use Spawn Probability=false)
  ]
  particleSpawnScript [
   &FN_SetSpawnGroup(Emitter Bone Count=4)
   &InitializeParticle(
    Color=<rgba 0 0 0 0>
    Lifetime=100
    Module.Write Color=true
    Module.Write Lifetime=false
    Module.Write Mass=true
    Module.Write Material Random=true
    Module.Write Position=false
    Module.Write Scale=false
    Module.Write SpriteRotation=false
    Module.Write SpriteSize=true
    Position=&SimulationPosition()
    Sprite Size=<50 50>
   )
   &CameraOffset(Camera Offset Amount=&UniformRangedFloat(Minimum=10, Maximum=40))
   /SetVariables(Particles.RandomFloat=&UniformRangedFloat(Minimum=0, Maximum=1)) {isSetParameters=true}
  ]
  particleUpdateScript [
   &SampleSkeletalMeshSkeleton(Mesh:NiagaraDataInterfaceSkeletalMesh=$User.SkeletalMesh, Specific Bone Index:NiagaraInt32=$Particles.SpawnGroup)
   &SkeletalMeshSkeletonLocation(Module.Inherit Velocity=false, Module.Use Sampled Bone as Ribbon ID=true)
   &ParticleState(Kill Particles When Lifetime Has Elapsed=false, Let Infinitely Lived Particles Die When Emitter Deactivates=true)
   &KillParticles(Kill Particles=/Game/Effects/Niagara/DynamicInputs/State/FN_IsExecutionStateActive())
   /SetVariables(Particles.Flicker:NiagaraFloat=&FloatFromCurve(
     FloatCurve=DataInterface() {propertyValues {Curve {keys [
         {interpMode=RCIM_Cubic, tangentMode=RCTM_User, time=-0.0046, value=1.0113}
         {interpMode=RCIM_Cubic, tangentMode=RCTM_User, time=0.0911, value=1.2315}
         {interpMode=RCIM_Cubic, tangentMode=RCTM_User, time=0.1896, value=0.8921}
         {interpMode=RCIM_Cubic, tangentMode=RCTM_User, time=0.2477, value=1.0085}
         {interpMode=RCIM_Cubic, tangentMode=RCTM_User, time=0.4771, value=1.1779}
         {interpMode=RCIM_Cubic, tangentMode=RCTM_User, time=0.5879, value=0.9278}
         {interpMode=RCIM_Cubic, tangentMode=RCTM_User, time=0.6577, value=1.1779}
         {interpMode=RCIM_Cubic, tangentMode=RCTM_User, time=0.8836, value=0.9099}
         {interpMode=RCIM_Cubic, tangentMode=RCTM_User, time=1, value=1}
        ]}}}
     CurveIndex=HlslExpression() {hlslExpression="fmod(Particles.RandomFloat + Particles.Age, 1.0)"}
    )) {isSetParameters=true}
   &Color(
    Color=<rgba 0.141 0.054 0.25 0.075>
    Scale Alpha=&Multiply_Float(A=$Particles.Flicker, B=1.25)
    Scale Color=&VectorFromFloat(Value=&Multiply_Float(A=$Particles.Flicker, B=2))
   )
   &ScaleSpriteSize(Scale Factor=&Vector2DFromFloat(Value=&Multiply_Float(A=$Particles.Flicker, B=1)))
  ]
  rendererClasses [&NiagaraSpriteRendererProperties]
  renderers [{rendererClass=&NiagaraSpriteRendererProperties}]
 }
}
summary {
 systemName=NS_Contrail_Raven
 emitters [
  {emitterName=Feathers_ChainedInheritance, rendererClasses [&NiagaraMeshRendererProperties]}
  {emitterName=Glow_ChainedInheritance, rendererClasses [&NiagaraSpriteRendererProperties]}
  {emitterName=FN_Contrails_Ribbon_v4, rendererClasses [&NiagaraRibbonRendererProperties]}
 ]
 userVariables [
  User.bIsFrontEnd {type=NiagaraBool}
  User.bIsFrontEndPreview {type=NiagaraBool}
  User.FEVec {type=Vector3f, description="", defaultValue {struct=Vector3f, value=<0 0 1000>}}
  User.PlayerSpeed {type=NiagaraFloat, defaultValue=0, description=""}
  User.RibbonIDOffset {type=NiagaraInt32}
  User.SkeletalMesh {
   type=NiagaraDataInterfaceSkeletalMesh
   defaultValue {
    struct=NiagaraExt_VariableValue_DataInterface
    value {dataInterface=&self:NiagaraDataInterfaceSkeletalMesh_0}
   }
  }
  User.SpawnRate {type=NiagaraFloat, defaultValue=360, description=""}
 ]
}
